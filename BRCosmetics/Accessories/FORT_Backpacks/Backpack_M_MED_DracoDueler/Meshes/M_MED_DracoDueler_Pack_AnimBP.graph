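# Animation Blueprint for the DracoDueler backpack, procedurally posing its pack bone to follow body movement.
# A cached base pose applies spine-driven offsets; when the character wears a tail, blends select skydive, float or ground corrections.
# The event graph derives their alphas from movement, jumping and spine rotation, plus socket locations for Look At nodes.
# It also toggles a "Hide Element 01" material parameter based on matching pickaxe, skydiving and front-end idle state.
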AnimBlueprint /BRCosmetics/Accessories/FORT_Backpacks/Backpack_M_MED_DracoDueler/Meshes/M_MED_DracoDueler_Pack_AnimBP
target_skeleton=&/M_MED_DracoDueler_Pack_Skeleton
graphs [
 AnimGraph {nodes [
   n1 = "Transform (Modify) Bone - Bone: pack"!(
    ComponentPose=Local To Component(LocalPose=Use cached pose 'base pose'())
    bAlphaBoolEnabled:bool=Get bIsSkyDiving()
    Translation="0,-2,8"
    Rotation="0,0,15"
    Scale
   ) -> (Pose) {
    alphaInputType=Bool
    rotationMode=BMM_Replace
    rotationSpace=BCS_BoneSpace
    translationMode=BMM_Additive
    translationSpace=BCS_BoneSpace
    alphaBoolBlend {blendInTime=0.2, blendOption=Linear, blendOutTime=0.2}
    boneToModify {boneName=pack}
    rotation {roll=15}
    translation {y=-2, z=8}
   }
   n2 = "Transform (Modify) Bone - Bone: pack"(ComponentPose=@n1, bAlphaBoolEnabled=Get bIsSkydiveDiveMode(), Translation="0,3,-3", Rotation="0,0,-15") {
    alphaInputType=Bool
    rotationMode=BMM_Additive
    rotationSpace=BCS_BoneSpace
    translationMode=BMM_Additive
    translationSpace=BCS_BoneSpace
    alphaBoolBlend {blendInTime=0.1, blendOption=Linear, blendOutTime=0.3}
    alphaScaleBiasClamp {bInterpResult=true, interpSpeedDecreasing=1, interpSpeedIncreasing=1}
    boneToModify {boneName=pack}
    rotation {roll=-15}
    translation {y=3, z=-3}
   }
   n3 = Component To Local(ComponentPose=@n2)
   n4 = "Transform (Modify) Bone - Bone: pack"!(
    ComponentPose=Local To Component(LocalPose=Copy Pose From Mesh())
    Alpha:float=0.4
    Translation=Get backpack_translation()
    Rotation=Get backpack_rotate()
    Scale
   ) -> (Pose) {
    rotationMode=BMM_Additive
    scaleMode=BMM_Additive
    scaleSpace=BCS_BoneSpace
    translationMode=BMM_Additive
    alphaBoolBlend {blendOption=Linear}
    boneToModify {boneName=pack}
   }
   n5 = "Transform (Modify) Bone - Bone: pack"!(
    ComponentPose=Look At - Bone: pack!(ComponentPose=@n4, Alpha=0.5, LookAtLocation:Vector=Get Tail Location() -> (Tail Location:Vector)) -> (Pose) {alphaBoolBlend {blendOption=Linear}, boneToModify {boneName=pack}, lookUp_Axis {axis=<0 1>}, lookAtTarget=~}
    bAlphaBoolEnabled:bool=Get bIsSkydiveFloating()
    Translation="0,-2,7"
    Rotation="0,0,4"
    Scale
   ) -> (Pose) {
    alphaInputType=Bool
    rotationMode=BMM_Additive
    rotationSpace=BCS_BoneSpace
    translationMode=BMM_Additive
    translationSpace=BCS_BoneSpace
    alphaBoolBlend {blendInTime=0.2, blendOption=Linear, blendOutTime=0.2}
    boneToModify {boneName=pack}
    rotation {roll=4}
    translation {y=-2, z=7}
   }
   n6 = "Transform (Modify) Bone - Bone: pack"!(
    ComponentPose=@n5
    Alpha:float=Get LeftMovementAlphaSkydiving() -> (LeftMovementAlphaSkydiving:double)
    Translation="2,-3,0"
    Rotation="3,13,1.5"
    Scale
   ) -> (Pose) {
    rotationMode=BMM_Additive
    rotationSpace=BCS_BoneSpace
    translationMode=BMM_Additive
    translationSpace=BCS_BoneSpace
    alphaBoolBlend {blendInTime=0.2, blendOption=Linear, blendOutTime=0.2}
    alphaScaleBiasClamp {bInterpResult=true, interpSpeedIncreasing=1}
    boneToModify {boneName=pack}
    rotation=<rot 3 13 1.5>
    translation=<2 -3>
   }
   n7 = "Transform (Modify) Bone - Bone: pack"(
    ComponentPose=@n6
    Alpha=Get RightMovementAlphaSkydiving() -> (RightMovementAlphaSkydiving:double)
    Translation="0,-3,0"
    Rotation="7,-10,-5"
   ) {
    rotationMode=BMM_Additive
    rotationSpace=BCS_BoneSpace
    translationMode=BMM_Additive
    translationSpace=BCS_BoneSpace
    alphaBoolBlend {blendInTime=0.2, blendOption=Linear, blendOutTime=0.2}
    alphaScaleBiasClamp {bInterpResult=true, interpSpeedDecreasing=1, interpSpeedIncreasing=1}
    boneToModify {boneName=pack}
    rotation=<rot 7 -10 -5>
    translation {y=-3}
   }
   n8 = "Transform (Modify) Bone - Bone: pack"!(
    ComponentPose=@n7
    bAlphaBoolEnabled:bool=Get parachute closed() -> (parachute closed:bool)
    Translation="0,-4,0"
    Rotation="0,-10,-15"
    Scale
   ) -> (Pose) {
    alphaInputType=Bool
    rotationMode=BMM_Additive
    rotationSpace=BCS_BoneSpace
    translationMode=BMM_Additive
    translationSpace=BCS_BoneSpace
    alphaBoolBlend {blendInTime=0.1, blendOption=Linear, blendOutTime=0.3}
    alphaScaleBiasClamp {bInterpResult=true, interpSpeedDecreasing=1, interpSpeedIncreasing=1}
    boneToModify {boneName=pack}
    rotation {roll=-15, yaw=-10}
    translation {y=-4}
   }
   n9 = Component To Local(ComponentPose=@n8)
   n10 = Get bIsSkydiveFloating()
   n11 = "Transform (Modify) Bone - Bone: pack"!(
    ComponentPose=Local To Component(LocalPose=Blend Poses by bool(BlendPose_0=@n9, BlendPose_1=@n3, bActiveValue=@n10) {blendType=Linear})
    Alpha:float
    Translation="0,0,0"
    Rotation="0,0,0"
    Scale="0.8,0.8,0.8"
   ) -> (Pose) {scaleMode=BMM_Replace, alphaBoolBlend {blendOption=Linear}, boneToModify {boneName=pack}, scale=<0.8 0.8 0.8>}
   n12 = Component To Local(ComponentPose=@n11)
   n13 = Get bIsSkyDiving()
   n14 = Use cached pose 'base pose'()
   n15 = "Transform (Modify) Bone - Bone: pack"!(
    ComponentPose=Look At - Bone: pack(ComponentPose=Local To Component(LocalPose=@n14), Alpha=0.2, LookAtLocation=Get Pelvis Location() -> (Pelvis Location:Vector)) {alphaBoolBlend {blendOption=Linear}, boneToModify {boneName=pack}, lookUp_Axis {axis=<0 1>}, lookAtTarget=~}
    Translation="0,0,0"
    Rotation="0,0,-10"
    Scale
   ) -> (Pose) {
    alphaInputType=Bool
    rotationMode=BMM_Additive
    rotationSpace=BCS_BoneSpace
    translationMode=BMM_Additive
    translationSpace=BCS_BoneSpace
    alphaBoolBlend {blendInTime=0.2, blendOption=Linear, blendOutTime=0.2}
    boneToModify {boneName=pack}
    rotation {roll=-10}
   }
   n16 = "Transform (Modify) Bone - Bone: pack"!(ComponentPose=@n15, Alpha:float=Get Start moving right() -> (Start moving right:double), Translation="1,-1,2", Rotation="5,8,0", Scale) -> (Pose) {
    rotationMode=BMM_Additive
    rotationSpace=BCS_BoneSpace
    translationMode=BMM_Additive
    translationSpace=BCS_BoneSpace
    alphaBoolBlend {blendInTime=0.1, blendOption=Linear, blendOutTime=0.3}
    alphaScaleBiasClamp {bInterpResult=true, interpSpeedDecreasing=6, interpSpeedIncreasing=3}
    boneToModify {boneName=pack}
    rotation {pitch=5, yaw=8}
    translation=<1 -1 2>
   }
   n17 = "Transform (Modify) Bone - Bone: pack"!(ComponentPose=@n16, bAlphaBoolEnabled:bool=Get Right mvmnt corr() -> (Right mvmnt corr:bool), Translation="1,-1,3", Rotation="4,6,5", Scale) -> (Pose) {
    alphaInputType=Bool
    rotationMode=BMM_Additive
    rotationSpace=BCS_BoneSpace
    translationMode=BMM_Additive
    translationSpace=BCS_BoneSpace
    alphaBoolBlend {blendInTime=0.2, blendOption=Linear, blendOutTime=0.2}
    boneToModify {boneName=pack}
    rotation=<rot 4 6 5>
    translation=<1 -1 3>
   }
   n18 = "Transform (Modify) Bone - Bone: pack"!(ComponentPose=@n17, Alpha:float=Get Start moving left() -> (Start moving left:double), Translation="-3,-1,2", Rotation="-5,-8,0", Scale) -> (Pose) {
    rotationMode=BMM_Additive
    rotationSpace=BCS_BoneSpace
    translationMode=BMM_Additive
    translationSpace=BCS_BoneSpace
    alphaBoolBlend {blendInTime=0.1, blendOption=Linear, blendOutTime=0.3}
    alphaScaleBiasClamp {bInterpResult=true, interpSpeedDecreasing=6, interpSpeedIncreasing=3}
    boneToModify {boneName=pack}
    rotation {pitch=-5, yaw=-8}
    translation=<-3 -1 2>
   }
   n19 = "Transform (Modify) Bone - Bone: pack"!(ComponentPose=@n18, bAlphaBoolEnabled:bool=Get Left mvmnt corr() -> (Left mvmnt corr:bool), Translation="-1,-1,3", Rotation="-4,-6,8", Scale) -> (Pose) {
    alphaInputType=Bool
    rotationMode=BMM_Additive
    rotationSpace=BCS_BoneSpace
    translationMode=BMM_Additive
    translationSpace=BCS_BoneSpace
    alphaBoolBlend {blendInTime=0.2, blendOption=Linear, blendOutTime=0.2}
    boneToModify {boneName=pack}
    rotation=<rot -4 -6 8>
    translation=<-1 -1 3>
   }
   n20 = "Transform (Modify) Bone - Bone: pack"(ComponentPose=@n19, bAlphaBoolEnabled=Get Back mvmnt corr() -> (Back mvmnt corr:bool), Translation="-3,-2,3", Rotation="12,0,8") {
    alphaInputType=Bool
    rotationMode=BMM_Additive
    rotationSpace=BCS_BoneSpace
    translationMode=BMM_Additive
    translationSpace=BCS_BoneSpace
    alphaBoolBlend {blendInTime=0.2, blendOption=Linear, blendOutTime=0.2}
    boneToModify {boneName=pack}
    rotation {pitch=12, roll=8}
    translation=<-3 -2 3>
   }
   n21 = "Transform (Modify) Bone - Bone: pack"!(ComponentPose=@n20, Alpha:float="Get  jump alpha pack"() -> (" jump alpha pack":double), Translation="0,0,0", Rotation="-10,0,0", Scale) -> (Pose) {
    rotationMode=BMM_Additive
    rotationSpace=BCS_BoneSpace
    translationMode=BMM_Additive
    translationSpace=BCS_BoneSpace
    alphaBoolBlend {blendInTime=0.1, blendOption=Custom, blendOutTime=0.6}
    alphaScaleBiasClamp {bInterpResult=true}
    boneToModify {boneName=pack}
    rotation {pitch=-10}
   }
   n22 = "Transform (Modify) Bone - Bone: pack"(ComponentPose=@n21, Alpha=Get targeting up left() -> (targeting up left:double), Translation="1.7,0,1", Rotation="5,0,0") {
    rotationMode=BMM_Additive
    rotationSpace=BCS_BoneSpace
    translationMode=BMM_Additive
    translationSpace=BCS_BoneSpace
    alphaBoolBlend {blendInTime=0.1, blendOption=Custom, blendOutTime=0.6}
    alphaScaleBiasClamp {bInterpResult=true, interpSpeedDecreasing=5, interpSpeedIncreasing=5}
    boneToModify {boneName=pack}
    rotation {pitch=5}
    translation {x=1.7, z=1}
   }
   n23 = Component To Local(ComponentPose=@n22)
   Output Pose(Result=Blend Poses by bool(
     BlendPose_0=Blend Poses by bool(BlendPose_0=@n12, BlendPose_1=@n23, bActiveValue=@n13) {blendType=Linear}
     BlendPose_1=Use cached pose 'base pose'()
     bActiveValue=Get Is wearing tail() -> (Is wearing tail:bool)
    ) {blendType=Linear}) {layerGroup=None}
   n24 = "Transform (Modify) Bone - Bone: pack"(ComponentPose=Local To Component(LocalPose=Input Pose()), Translation=Get backpack_translation(), Rotation=Get backpack_rotate()) {
    rotationMode=BMM_Additive
    scaleMode=BMM_Additive
    scaleSpace=BCS_BoneSpace
    translationMode=BMM_Additive
    alphaBoolBlend {blendOption=Linear}
    boneToModify {boneName=pack}
   }
   Save cached pose 'base pose'(Pose=Component To Local(ComponentPose=@n24)) {cacheName=base pose}
   Reroute Node(InputPin:Wildcard) -> (OutputPin:Wildcard)
  ]}
 EventGraph {nodes [
   Event OnParachuteClosed()
   > Set parachute closed(parachute closed:bool=true) -> (Output_Get:bool)
   > Delay(Duration=3)
   > Set parachute closed(parachute closed=false)
   n1 = Event OnCharacterCustomizationLoaded()
   > n2 = Sequence!(execute) -> (then_0, then_1, then_2:Exec)
   n3 = Get Dynamic Material Instance(execute=[@n8.then, @n2.then_0, @n11.then_1], SkeletalMeshComponent=GetPartSkeletalMeshComponent())
   n4 = PickaxeItemDefinitionHasAnyMetaTag(Tags=Make Array("[0]"=(TagName="Cosmetics.MetaTags.DracoDueler")), bRequiresEquipped=true)
   > n5 = Branch(Condition=@n4)
   > n6 = Branch(Condition=Get bIsSkyDiving())
   SetScalarParameterValue(execute=[@n9.then, @n7.else, @n10.else, @n6.then], self=@n3, ParameterName=Hide Element 01, Value=0)
   SetScalarParameterValue(execute=[@n7.then, @n10.then, @n6.else], self=@n3, ParameterName=Hide Element 01, Value=1)
   n7 = Branch(execute=@n5.else, Condition=Get FE Idle Active() -> (FE Idle Active:bool))
   Set Skel Mesh(execute=@n2.then_1, Skel Mesh=@n1.CustomBaseMesh)
   Set Tail Mesh(
    execute=@n2.then_2
    Tail Mesh:Skeletal Mesh Component Budgeted Object Reference=Get Array Item with EFort Custom Part Type(Array=@n1.CustomMeshParts, PartType=MiscOrTail)
   ) -> (Output_Get:Skeletal Mesh Component Budgeted Object Reference)
   n8 = Event OnWeaponChanged()
   n9 = Event OnBeginSkydiving()
   Event OnEndSkydiving()
   > n10 = Branch(Condition=@n4)
   Event BlueprintUpdateAnimation()
   > n11 = Sequence!(execute) -> (then_0, then_1, then_2, then_3:Exec, then_4:Exec, then_5:Exec, then_6:Exec, then_7:Exec, then_8:Exec)
   n12 = Sequence!(execute=@n11.then_3) -> (then_0, then_1, then_2)
   n13 = OR Boolean(A=Get bIsSkydiveFloating(), B=Get bIsSkyDiving())
   n14 = Get bIsFalling()
   n15 = Get bIsJumping()
   n16 = Branch(execute=@n12.then_2, Condition=AND Boolean(A=OR Boolean(A=@n15, B=@n14), B=NOT Boolean(A=@n13)))
   > Do Once(Reset=@n17.then)
   > "Set  jump alpha pack"(" jump alpha pack":double=1) -> (Output_Get:double)
   > Delay(Duration=0.3)
   > "Set  jump alpha pack"(" jump alpha pack"=0)
   n17 = "Set  jump alpha pack"(execute=@n16.else, " jump alpha pack"=0)
   n18 = Get bIsJumping()
   n19 = Branch(execute=@n12.then_1, Condition=AND Boolean(A=NOT Boolean(A=@n18), B=Get Left mvmnt corr()))
   > Do Once(Reset=@n20.then)
   > Set Start moving left(Start moving left:double=1) -> (Output_Get:double)
   > Delay(Duration=0.3)
   > Set Start moving left(Start moving left=0)
   n20 = Set Start moving left(execute=@n19.else, Start moving left=0)
   n21 = Get bIsJumping()
   n22 = Branch(execute=@n12.then_0, Condition=AND Boolean(A=NOT Boolean(A=@n21), B=Get Right mvmnt corr()))
   > Do Once(Reset=@n23.then)
   > Set Start moving right(Start moving right:double=1) -> (Output_Get:double)
   > Delay(Duration=0.3)
   > Set Start moving right(Start moving right=0)
   n23 = Set Start moving right(execute=@n22.else, Start moving right=0)
   n24 = Sequence!(execute=@n11.then_4) -> (then_0, then_1, then_2, then_3:Exec)
   n25 = Get Skel Mesh()
   > Is Valid(InputObject=@n25)
   > Set Pelvis Location(Pelvis Location:Vector=GetSocketLocation!(self:Scene Component Object Reference=@n25, InSocketName:Name=pelvis) -> (ReturnValue:Vector)) -> (Output_Get:Vector)
   n26 = Get Skel Mesh()
   Is Valid(exec=@n24.then_1, InputObject=@n26)
   > Set Tail Location(Tail Location:Vector=GetSocketLocation(self=@n26, InSocketName=Tail)) -> (Output_Get:Vector)
   n27 = Get Tail Mesh() -> (Tail Mesh:Skeletal Mesh Component Budgeted Object Reference)
   Is Valid(exec=@n24.then_2, InputObject=@n27)
   > Set Tail(Tail:Vector=GetSocketLocation(self=@n27, InSocketName=C_Tail_A_4_Jnt)) -> (Output_Get:Vector)
   n28 = BreakTransform(InTransform=GetDeltaTransformFromRefPose(self=GetOwnerSkeletalMeshComponent(), BoneName=spine_05, BaseName=spine_01))
   Set targeting up left(execute=@n11.then_6, targeting up left:double=MapRangeClamped(Value=BreakRotator(InRot=@n28.Rotation), InRangeA=5, InRangeB=10)) -> (Output_Get:double)
   Branch(execute=@n11.then_8, Condition=Get bIsInFrontEnd())
   n29 = Make Array("[0]"=(TagName="Cosmetics.MetaTags.DracoDueler"))
   n30 = GetCurveValue(self=GetOwnerFrontEndAnimInstance(), CurveName=FrontEnd_Idle_Bypass)
   > n31 = Branch(Condition=AND Boolean(A=CharacterItemDefinitionHasAnyMetaTag(Tags=@n29), B="float >= float"(A=@n30, B=1)))
   > Set FE Idle Active(FE Idle Active:bool=true) -> (Output_Get:bool)
   Set FE Idle Active(execute=@n31.else, FE Idle Active=false)
   n32 = GetOwnerSkeletalMeshComponent()
   n33 = BreakRotator(InRot=BreakTransform(InTransform=GetDeltaTransformFromRefPose(self=@n32, BoneName=spine_05, BaseName=spine_01)).Rotation)
   n34 = MapRangeClamped(Value=@n33.Pitch, InRangeA=30, InRangeB=-30, OutRangeA=-30, OutRangeB=30)
   n35 = MapRangeClamped(Value=@n33.Pitch, InRangeA=0, InRangeB=-30, OutRangeA=-5, OutRangeB=25)
   n36 = OR Boolean(A=Get bIsSkydiveFloating(), B=Get bIsSkyDiving())
   Set backpack_rotate(
    execute=@n11.then_0
    backpack_rotate=MakeRotator(Roll=MapRangeClamped(Value=@n33.Yaw, InRangeA=0, InRangeB=-30, OutRangeA=5, OutRangeB=25), Pitch=SelectFloat(A=@n35, B=@n34, bPickA=@n36))
   )
   > Set backpack_translation(backpack_translation=MakeVector(Y=MapRangeClamped(Value=@n33.Yaw, InRangeA=0, InRangeB=-30, OutRangeB=-6.5)))
   Set Is wearing tail(execute=@n11.then_7, Is wearing tail:bool=CheckForTail!(self:Custom Character Part Anim Instance Object Reference) -> (bHasTail:bool)) -> (Output_Get:bool)
   n37 = TryGetPawnOwner()
   n38 = Dot Product!(
    A:Vector=GetVelocity!(self:Actor Object Reference=@n37) -> (ReturnValue:Vector)
    B:Vector=GetActorRightVector!(self:Actor Object Reference=@n37) -> (ReturnValue:Vector)
   ) -> (ReturnValue:double)
   Set RightMovementAlphaSkydiving(execute=@n11.then_5, RightMovementAlphaSkydiving:double=MapRangeClamped(Value=@n38, InRangeA=0, InRangeB=230)) -> (Output_Get:double)
   > Set LeftMovementAlphaSkydiving(LeftMovementAlphaSkydiving:double=MapRangeClamped(Value=@n38, InRangeA=0, InRangeB=-230)) -> (Output_Get:double)
   n39 = Get bIsCrouching()
   n40 = Get PawnMovementDirectionAzimuth() -> (PawnMovementDirectionAzimuth:float)
   n41 = "float > float"(A=@n40, B=-45)
   n42 = "float < float"(A=@n40, B=45)
   Set Is moving forward(
    execute=@n11.then_2
    Is moving forward:bool=AND Boolean!(A=AND Boolean!(A=@n41, B=@n42) -> (ReturnValue), B=Get bIsMoving2D(), C:bool=NOT Boolean(A=@n39)) -> (ReturnValue)
   ) -> (Output_Get:bool)
   n43 = Get bIsCrouching()
   n44 = OR Boolean!(A=Get bIsSkydiveFloating(), B=Get bIsSkyDiving(), C:bool=Get bIsTargeting(), D:bool=Get bIsJumping(), E:bool=Get bIsFalling()) -> (ReturnValue)
   n45 = Get PawnMovementDirectionAzimuth()
   n46 = "float > float"(A=@n45, B=25)
   n47 = "float < float"(A=@n45, B=140)
   > Set Right mvmnt corr(Right mvmnt corr:bool=AND Boolean!(A=AND Boolean!(A=@n46, B=@n47) -> (ReturnValue), B=Get bIsMoving2D(), C:bool=NOT Boolean(A=@n43), D:bool=NOT Boolean(A=@n44)) -> (ReturnValue)) -> (Output_Get:bool)
   n48 = Get bIsCrouching()
   n49 = OR Boolean(A=Get bIsSkydiveFloating(), B=Get bIsSkyDiving(), C=Get bIsTargeting(), D=Get bIsJumping(), E=Get bIsFalling())
   n50 = Get PawnMovementDirectionAzimuth()
   n51 = "float > float"(A=@n50, B=-140)
   n52 = "float < float"(A=@n50, B=-25)
   > Set Left mvmnt corr(Left mvmnt corr:bool=AND Boolean!(A=AND Boolean!(A=@n51, B=@n52) -> (ReturnValue), B=Get bIsMoving2D(), C:bool=NOT Boolean(A=@n48), D:bool=NOT Boolean(A=@n49)) -> (ReturnValue)) -> (Output_Get:bool)
   n53 = Get bIsCrouching()
   n54 = OR Boolean(A=Get bIsSkydiveFloating(), B=Get bIsSkyDiving(), C=Get bIsTargeting(), D=Get bIsJumping(), E=Get bIsFalling())
   n55 = Get PawnMovementDirectionAzimuth()
   n56 = AND Boolean(A="float > float"(A=@n55, B=140), B="float < float"(A=@n55, B=180))
   n57 = AND Boolean(A="float > float"(A=@n55, B=-180), B="float < float"(A=@n55, B=-140))
   > Set Back mvmnt corr(Back mvmnt corr:bool=AND Boolean!(A=OR Boolean!(A=@n56, B=@n57) -> (ReturnValue), B=Get bIsMoving2D(), C:bool=NOT Boolean(A=@n53), D:bool=NOT Boolean(A=@n54)) -> (ReturnValue)) -> (Output_Get:bool)
   n58 = NOT Boolean(A=OR Boolean!(A=Get bIsSkydiveFloating(), B=Get bIsSkyDiving(), C:bool=Get bIsTargeting()) -> (ReturnValue))
   n59 = OR Boolean!(A=Get bIsJumping(), B=Get bIsFalling()) -> (ReturnValue)
   > "Set  jump alpha pack"(" jump alpha pack"=SelectFloat(A=1, B=0, bPickA=AND Boolean!(A=@n59, B=@n58) -> (ReturnValue)))
   Reroute Node()
  ]}
]
variables [
 " jump alpha pack" {type=number}
 Back mvmnt corr {type=boolean}
 backpack_rotate {type=Rotator}
 backpack_translation {type=Vector}
 FE Idle Active {type=boolean}
 Is moving forward {type=boolean}
 Is wearing tail {type=boolean}
 Left mvmnt corr {type=boolean}
 LeftMovementAlphaSkydiving {type=number}
 parachute closed {type=boolean}
 Pelvis Location {type=Vector}
 Right mvmnt corr {type=boolean}
 RightMovementAlphaSkydiving {type=number}
 Skel Mesh {type=SkeletalMeshComponent}
 Start moving left {type=number}
 Start moving right {type=number}
 Tail {type=Vector}
 Tail Location {type=Vector}
 Tail Mesh {type=SkeletalMeshComponentBudgeted}
 targeting up left {type=number}
]
